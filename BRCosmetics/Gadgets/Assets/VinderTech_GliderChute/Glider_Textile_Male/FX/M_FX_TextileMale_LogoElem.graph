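# Unlit emissive material for Niagara mesh particles, used for a logo element in the Textile Male glider effects.
# A color parameter forms the base glow, added to a fake reflection from a beam-distort texture sampled by reflection vector XY and scaled by color alpha.
# A rotated, panning gradient adds a bright color-tinted sweep, timed by Time plus the TimeOffset dynamic parameter.
# The sum feeds MF_FX_EmissiveAdjustments.
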Material /BRCosmetics/Gadgets/Assets/VinderTech_GliderChute/Glider_Textile_Male/FX/M_FX_TextileMale_LogoElem
graph {
 properties {
  bUsedWithNiagaraMeshParticles=true
  pixelDepthOffsetMode=PDOM_Legacy
  previewMesh=&/Textile_Male_Logo_1
  shadingModel=MSM_Unlit
  usedShadingModels=Unlit
 }
 nodes [
  CameraVectorWS()
  n1 = ReflectionVectorWS()
  MaterialFunctionCall!(Vector=@n1) {materialFunction=/Engine/Functions/Engine_MaterialFunctions01/Coordinates/LongLatToUV}
  n2 = VectorParameter() {parameterName=Color, defaultValue=<rgba 0.1719 0.1063 0 1>}
  n3 = Add(A=Subtract(A=ComponentMask(None=@n1) {g=true, r=true}) {constB=0.5}) {constB=0.5}
  n4 = Multiply(
   A=@n2.RGB
   B=Multiply(A=TextureSample(UVs=@n3) {texture=/Game/Global/Textures/Default/Masks_and_Gradients/T_BlasterBeamDistort}.RGB, B=@n2.A) {constB=3}
  )
  n5 = Multiply(A=Rotator(Coordinate=TextureCoordinate(), Time=Constant() {r=1.5})) {constB=4}
  n6 = Add(A=Time(), B=DynamicParameter() {paramNames [TimeOffset,Param2,Param3,Param4]}.TimeOffset)
  n7 = Multiply(A=TextureSample(UVs=Panner(Coordinate=@n5, Time=@n6) {bFractionalPart=true, speedY=4}) {texture=/Game/Effects/Fort_Effects/Textures/Grad/T_FX_LinearSweep_Gradient}.R) {constB=6}
  MaterialFunctionCall(Emissive=Add(A=Add(A=@n4, B=@n2.RGB), B=Multiply(A=@n2.RGB, B=@n7))) {materialFunction=&MF_FX_EmissiveAdjustments}
  ParticleColor()
  TextureSample()
 ]
}
