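# Additive, two-sided Niagara mesh-particle material for sparkles on the Textile Male VinderTech glider.
# Particle color feeds an emissive-adjustment function.
# A grayscale cross-flare texture spins through a time-driven Rotator, with vertex red staggering each sparkle's phase.
# Centered UVs become camera-space vectors that are transformed, normalized and multiplied by a Scale parameter, a billboard-style expansion, though the graph shows no material output for it.
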Material /BRCosmetics/Gadgets/Assets/VinderTech_GliderChute/Glider_Textile_Male/FX/M_TextileMale_Sparkle
graph {
 properties {
  bUsedWithNiagaraMeshParticles=true
  blendMode=BLEND_Additive
  pixelDepthOffsetMode=PDOM_Legacy
  previewMesh=&/FX_SM_TextileMale_Sparkles_Logo_1
  twoSided=true
  usedShadingModels=Default Lit
 }
 nodes [
  MaterialFunctionCall(Emissive=ParticleColor().RGB) {materialFunction=&MF_FX_EmissiveAdjustments}
  n1 = TextureCoordinate()
  n2 = Transform(None=AppendVector(A=Subtract(A=@n1) {constB=0.5}, B=Constant())) {transformSourceType=TRANSFORMSOURCE_Camera}
  Multiply(A=Normalize(VectorInput=@n2), B=ScalarParameter() {defaultValue=121.8645, parameterName=Scale})
  n3 = VertexColor()
  n4 = Time()
  TextureSampleParameter2D(UVs=Rotator(Time=Add(A=@n3.R, B=@n4)) {speed=0.869}) {
   parameterName=Texture
   samplerSource=SSM_Clamp_WorldGroupSettings
   samplerType=SAMPLERTYPE_Grayscale
   texture=/Game/Effects/Fort_Effects/Textures/Flares/T_Cross_Flare
  }
  Transform(None=@n2) {transformSourceType=TRANSFORMSOURCE_View}
 ]
}
